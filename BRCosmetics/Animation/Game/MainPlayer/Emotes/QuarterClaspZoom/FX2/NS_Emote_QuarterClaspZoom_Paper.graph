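# Niagara effect for the QuarterClaspZoom emote that throws short bursts of colored paper mesh particles, positioned on a skeletal mesh previewed with the cap gadget.
# A change in User.BurstPaper triggers spawning, and its integer value selects spawn count and position offset.
# NE_Paper covers steps 3–6 and NE_Paper02 step 7, with differing spawn shapes.
# Pieces launch outward, tumble under gravity and drag, and scale in and out by curve.
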NiagaraSystem /BRCosmetics/Animation/Game/MainPlayer/Emotes/QuarterClaspZoom/FX2/NS_Emote_QuarterClaspZoom_Paper
dependencies {
 usedDataInterfaces [&NiagaraDataInterfaceArrayColor, &NiagaraDataInterfaceSkeletalMesh, &NiagaraDataInterfaceCurve]
 usedDynamicInputs [&MergeHelper_FloatSwitch, &SpawnGroup]
 usedModules [&SystemState]
}
emitters {
 NE_Paper {
  emitterUpdateScript [
   &EmitterState(Life Cycle Mode=::NewEnumerator1, Loop Behavior=::NewEnumerator1, Loop Duration Mode=::NewEnumerator1, UseLoopDelay=false)
   /Game/Effects/Niagara/Modules/Logic/UpdateValueStateChange(InputIntParam:NiagaraInt32=&MakeIntFromFloat(Float=$User.BurstPaper, Conversion Mode=::NewEnumerator1))
   &self:TriggerSpawn_04(
    Spawn Count:NiagaraInt32=&self:Merge_Helper_Int_Switch_15(
     0:NiagaraInt32=0
     1:NiagaraInt32=0
     2:NiagaraInt32=0
     3:NiagaraInt32=2
     4:NiagaraInt32=1
     5:NiagaraInt32=1
     6:NiagaraInt32=2
     7:NiagaraInt32=0
     8:NiagaraInt32=?
     9:NiagaraInt32=?
     10:NiagaraInt32=?
     11:NiagaraInt32=?
     12:NiagaraInt32=?
     13:NiagaraInt32=?
     14:NiagaraInt32=?
     Index:NiagaraInt32=&MakeIntFromFloat(Float=$User.BurstPaper, Conversion Mode=::NewEnumerator1)
     15:NiagaraInt32=?
     Num:NiagaraInt32=6
    )
    Spawn Group:NiagaraInt32=&MakeIntFromFloat(Float=$User.BurstPaper, Conversion Mode=::NewEnumerator1)
    Trigger:NiagaraBool=$Emitter.UpdateValueStateChange.Changed
   ) {module=TriggerSpawn_04}
  ]
  particleSpawnScript [
   /SetVariables(Particles.IndexIn:NiagaraFloat=$User.BurstPaper) {isSetParameters=true, module=SetVariables_#1}
   /SetVariables(Particles.ExecutionIndex:NiagaraFloat=&ReturnNormalizedExecIndex()) {isSetParameters=true, module=SetVariables_#2}
   &InitializeParticle(
    Color=&SelectLinearColorFromArray(
     Array Sampling Mode=::NewEnumerator0
     Direct Array Index=?
     Color Selection Array=DataInterface() {propertyValues {ColorData [<rgba 0 1 0.1093 1>, <rgba 0 0.5229 1 1>, <rgba 0.0999 0 1 1>, <rgba 1 0.7012 0 1>]}}
     Random Seed=0
    )
    Lifetime=&RandomRangeFloat(Minimum=0.25, Maximum=0.5)
    Mesh Scale Mode=::NewEnumerator3
    Mesh Uniform Scale=1.2
    Sprite Size Mode=::NewEnumerator1
    Uniform Sprite Size Max=5
    Uniform Sprite Size Min=3
    UsePositionOffset=true
    Write Material Random=true
   )
   &InitialMeshOrientation(Use Rotation Vector=false)
   &ShapeLocation(
    MaskSpawnGroup=false
    Override Seed=false
    Radius Position:NiagaraFloat=&RandomRangeFloat(Minimum=0.5, Maximum=1)
    Ring / Disc Distribution Mode:ENiagaraSphereDistributionMode=::NewEnumerator1
    Ring Radius=2
    Rotation Mode=::NewEnumerator2
    Shape Primitive=::NewEnumerator4
    U Position=&Multiply_Float(A=$Particles.ExecutionIndex, B=&RandomRangeFloat(Minimum=0.5, Maximum=0.6))
    UseOwnerScale=true
    Yaw / Pitch / Roll=<0.9 0 0>
   ) {module=ShapeLocation001}
   &AddVelocity(
    Module.Use Falloff Distance=false
    Origin Offset=<0 0 -1>
    Override Seed=false
    Velocity Mode=::NewEnumerator1
    Velocity Speed=&RandomRangeFloat(Minimum=200, Maximum=210)
   )
   &SkeletalMeshLocation(
    Bone / Socket Index=&MakeIntFromFloat(Float=$Particles.IndexIn, Conversion Mode=::NewEnumerator1)
    Bone Sampling Mode=::NewEnumerator4
    MaskSpawnGroup=false
    Sampled Position Coordinate Space=::Local
    Sampled Position Offset=&MergeHelper_VectorSwitch(
     0=<0 0 0>
     1=<-30 20 20>
     2=<0 0 20>
     3=<0 0 15>
     4=<0 0 15>
     5=<-25 0 30>
     6=<-25 0 30>
     7=<0 0 20>
     Index=&MakeIntFromFloat(Float=$User.BurstPaper, Conversion Mode=::NewEnumerator1)
     Num=::NewEnumerator6
    )
    Skeletal Mesh=DataInterface() {propertyValues {
      DefaultMesh=null
      ExcludeBoneName=root
      OptionalComponentClass=null
      PreviewMesh=/BRCosmetics/Animation/Props/Gadgets_QuarterClaspZoom_Cap/Meshes/QuarterClaspZoom_Cap_Gadget
      SkinningMode=SkinOnTheFly
      SoftSourceActor=null
      SourceMode=Default
      UvSetIndex=0
      WholeMeshLOD=-1
      bExcludeBone=false
      bReadDeformedGeometry=true
      bRequireCurrentFrameData=true
      ComponentTags []
      FilteredBones [
       QuarterClaspZoom_Cap_LOD0_JNT
       QuarterClaspZoom_Cap_LOD0_JNT
       QuarterClaspZoom_Cap_LOD0_JNT
       QuarterClaspZoom_Cap_LOD0_JNT
       QuarterClaspZoom_Cap_LOD0_JNT
       QuarterClaspZoom_Cap_LOD0_JNT
       QuarterClaspZoom_Cap_LOD0_JNT
       QuarterClaspZoom_Cap_LOD0_JNT
      ]
      FilteredSockets []
      MeshUserParameter {parameter {name=None, typeDefHandle {registeredTypeIndex=1}, varData []}}
      SamplingRegions []
     }}
   )
  ]
  particleUpdateScript [
   &ParticleState()
   &ScaleMeshSize(Scale Factor=&VectorFromFloat(Value=&FloatFromCurve(FloatCurve=DataInterface() {propertyValues {Curve {defaultValue=1, keys [{tangentMode=RCTM_User, time=0, value=0}, <at 0.05 1>, <at 0.5 1>, {tangentMode=RCTM_User, time=1, value=0}]}}})))
   &UpdateMeshOrientation(
    Rotation Rate=&RandomRangeFloat(Minimum=1, Maximum=2)
    Rotation Vector=&RandomRangeVector(Minimum=<-1 -1 -1>, Maximum=<1 1 1>)
   )
   &GravityForce()
   &Drag(Drag=2, Use Linear Drag=true, Use Rotational Drag=false)
   &SolveForcesAndVelocity(Clamp Velocity=false, Limit Acceleration=false)
  ]
  rendererClasses [&NiagaraMeshRendererProperties]
  renderers [{rendererClass=&NiagaraMeshRendererProperties, rendererIndex=0}]
 }
 NE_Paper02 {
  emitterUpdateScript [
   &EmitterState(Life Cycle Mode=::NewEnumerator1, Loop Behavior=::NewEnumerator1, Loop Duration Mode=::NewEnumerator1, UseLoopDelay=false)
   &UpdateValueStateChange(InputIntParam:NiagaraInt32=&MakeIntFromFloat(Float=$User.BurstPaper, Conversion Mode=::NewEnumerator1))
   &self:TriggerSpawn_04(
    Spawn Count:NiagaraInt32=&self:Merge_Helper_Int_Switch_15(
     0:NiagaraInt32=0
     1:NiagaraInt32=0
     2:NiagaraInt32=0
     3:NiagaraInt32=0
     4:NiagaraInt32=0
     5:NiagaraInt32=0
     6:NiagaraInt32=0
     7:NiagaraInt32=&RandomRangeInt(Minimum=2, Maximum=3)
     8:NiagaraInt32=?
     9:NiagaraInt32=?
     10:NiagaraInt32=?
     11:NiagaraInt32=?
     12:NiagaraInt32=?
     13:NiagaraInt32=?
     14:NiagaraInt32=?
     Index:NiagaraInt32=&MakeIntFromFloat(Float=$User.BurstPaper, Conversion Mode=::NewEnumerator1)
     15:NiagaraInt32=?
     Num:NiagaraInt32=6
    )
    Spawn Group:NiagaraInt32=&MakeIntFromFloat(Float=$User.BurstPaper, Conversion Mode=::NewEnumerator1)
    Trigger:NiagaraBool=$Emitter.UpdateValueStateChange.Changed
   ) {module=TriggerSpawn_04}
  ]
  particleSpawnScript [
   /SetVariables(Particles.IndexIn:NiagaraFloat=$User.BurstPaper) {isSetParameters=true, module=SetVariables_#1}
   /SetVariables(Particles.ExecutionIndex:NiagaraFloat=&ReturnNormalizedExecIndex()) {isSetParameters=true, module=SetVariables_#2}
   &InitializeParticle(
    Color=&SelectLinearColorFromArray(
     Array Sampling Mode=::NewEnumerator0
     Direct Array Index=?
     Color Selection Array=DataInterface() {propertyValues {ColorData [<rgba 0 1 0.1093 1>, <rgba 0 0.5229 1 1>, <rgba 1 0.7012 0 1>, <rgba 0.0999 0 1 1>]}}
     Random Seed=0
    )
    Lifetime=&RandomRangeFloat(Minimum=0.25, Maximum=0.5)
    Mesh Scale Mode=::NewEnumerator3
    Sprite Size Mode=::NewEnumerator1
    Uniform Sprite Size Max=5
    Uniform Sprite Size Min=3
    UsePositionOffset=true
    Write Material Random=true
   )
   &InitialMeshOrientation(Use Rotation Vector=false)
   &ShapeLocation(
    Disc Coverage=0.3
    MaskSpawnGroup=false
    Override Seed=false
    Ring Radius=2
    Rotation Mode=::NewEnumerator2
    Shape Primitive=::NewEnumerator4
    U Distribution=0.5
    UseOwnerScale=true
    Yaw / Pitch / Roll=<0.5 0 0>
   ) {module=ShapeLocation001}
   &AddVelocity(
    Module.Use Falloff Distance=false
    Origin Offset=&RandomRangeVector(Minimum=<0 0 -2>, Maximum=<0 0 -1.5>)
    Override Seed=false
    Velocity Mode=::NewEnumerator1
    Velocity Speed=&RandomRangeFloat(Minimum=200, Maximum=210)
   )
   &SkeletalMeshLocation(
    Bone / Socket Index=&MakeIntFromFloat(Float=$Particles.IndexIn, Conversion Mode=::NewEnumerator1)
    Bone Sampling Mode=::NewEnumerator4
    MaskSpawnGroup=false
    Sampled Position Coordinate Space=::Local
    Sampled Position Offset=&MergeHelper_VectorSwitch(
     0=<0 0 0>
     1=<-30 20 20>
     2=<0 0 20>
     3=<0 0 20>
     4=<0 0 20>
     5=<0 0 20>
     6=<-30 0 20>
     7=<0 10 20>
     Index=&MakeIntFromFloat(Float=$User.BurstPaper, Conversion Mode=::NewEnumerator1)
     Num=::NewEnumerator6
    )
    Skeletal Mesh=DataInterface() {propertyValues {
      DefaultMesh=null
      ExcludeBoneName=root
      OptionalComponentClass=null
      PreviewMesh=&QuarterClaspZoom_Cap_Gadget
      SkinningMode=SkinOnTheFly
      SoftSourceActor=null
      SourceMode=Default
      UvSetIndex=0
      WholeMeshLOD=-1
      bExcludeBone=false
      bReadDeformedGeometry=true
      bRequireCurrentFrameData=true
      ComponentTags []
      FilteredBones [
       QuarterClaspZoom_Cap_LOD0_JNT
       QuarterClaspZoom_Cap_LOD0_JNT
       QuarterClaspZoom_Cap_LOD0_JNT
       QuarterClaspZoom_Cap_LOD0_JNT
       QuarterClaspZoom_Cap_LOD0_JNT
       QuarterClaspZoom_Cap_LOD0_JNT
       QuarterClaspZoom_Cap_LOD0_JNT
       QuarterClaspZoom_Cap_LOD0_JNT
      ]
      FilteredSockets []
      MeshUserParameter {parameter {name=None, typeDefHandle {registeredTypeIndex=1}, varData []}}
      SamplingRegions []
     }}
   )
  ]
  particleUpdateScript [
   &ParticleState()
   &ScaleMeshSize(Scale Factor=&VectorFromFloat(Value=&FloatFromCurve(FloatCurve=DataInterface() {propertyValues {Curve {defaultValue=1, keys [{tangentMode=RCTM_User, time=0, value=0}, <at 0.05 1>, <at 0.5 1>, {tangentMode=RCTM_User, time=1, value=0}]}}})))
   &UpdateMeshOrientation(
    Rotation Rate=&RandomRangeFloat(Minimum=1, Maximum=2)
    Rotation Vector=&RandomRangeVector(Minimum=<-1 -1 -1>, Maximum=<1 1 1>)
   )
   &GravityForce()
   &Drag(Drag=2, Use Linear Drag=true, Use Rotational Drag=false)
   &SolveForcesAndVelocity(Clamp Velocity=false, Limit Acceleration=false)
  ]
  rendererClasses [&NiagaraMeshRendererProperties]
  renderers [{rendererClass=&NiagaraMeshRendererProperties, rendererIndex=0}]
 }
}
summary {
 systemName=NS_Emote_QuarterClaspZoom_Paper
 emitters [
  {emitterName=NE_Paper, rendererClasses [&NiagaraMeshRendererProperties]}
  {emitterName=NE_Paper02, rendererClasses [&NiagaraMeshRendererProperties]}
 ]
 userVariables [
  User.bIsFrontEnd {type=NiagaraBool}
  User.bIsFrontEndPreview {type=NiagaraBool}
  User.BurstPaper {type=NiagaraFloat, defaultValue=0, description=""}
 ]
}
