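# Niagara system for the QuarterClaspZoom emote that throws small, short-lived colored paper mesh particles from the cap gadget.
# A single emitter, NE_Paper, spawns at a steady rate, using skeletal mesh sampling filtered to the cap bone.
# Each piece takes one of four array colors and launches fast with randomized speed.
# Gravity and drag shape its flight while it tumbles, and a curve pops its scale in and out.
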NiagaraSystem /BRCosmetics/Animation/Game/MainPlayer/Emotes/QuarterClaspZoom/FX2/NS_Emote_QuarterClaspZoom_Paper1
dependencies {
 usedDataInterfaces [&NiagaraDataInterfaceArrayColor, &NiagaraDataInterfaceSkeletalMesh, &NiagaraDataInterfaceCurve]
 usedDynamicInputs [&MergeHelper_FloatSwitch, &SpawnGroup]
 usedModules [&SystemState]
}
emitters {NE_Paper {
  emitterUpdateScript [
   &EmitterState(Life Cycle Mode=::NewEnumerator1, Loop Behavior=::NewEnumerator1, Loop Delay=3, Loop Duration=2, UseLoopDelay=false)
   &SpawnRate(SpawnRate=5, Use Spawn Probability=false)
  ]
  particleSpawnScript [
   &InitializeParticle(
    Color=&SelectLinearColorFromArray(
     Array Sampling Mode=::NewEnumerator0
     Direct Array Index=?
     Color Selection Array=DataInterface() {propertyValues {ColorData [<rgba 0 1 0.1093 1>, <rgba 0 0.5229 1 1>, <rgba 0.0999 0 1 1>, <rgba 1 0.7012 0 1>]}}
     Random Seed=0
    )
    Lifetime=&RandomRangeFloat(Minimum=0.4, Maximum=0.45)
    Mesh Scale Mode=::NewEnumerator3
    Position Offset=<0 0 10>
    Sprite Size Mode=::NewEnumerator1
    Uniform Sprite Size Max=5
    Uniform Sprite Size Min=3
    UsePositionOffset=true
    Write Material Random=true
   )
   &InitialMeshOrientation(Use Rotation Vector=false)
   &ShapeLocation(Hemisphere Distribution=<1 -0.5>, MaskSpawnGroup=false, Override Seed=false, Sphere Radius=1, UseOwnerScale=true) {module=ShapeLocation001}
   &AddVelocity(
    Module.Use Falloff Distance=false
    Override Seed=false
    Velocity Mode=::NewEnumerator1
    Velocity Speed=&RandomRangeFloat(Minimum=280, Maximum=290)
   )
   &SkeletalMeshLocation(
    Bone Sampling Mode=::NewEnumerator4
    MaskSpawnGroup=false
    Sampled Position Coordinate Space=::Local
    Skeletal Mesh=DataInterface() {propertyValues {
      DefaultMesh=null
      ExcludeBoneName=root
      OptionalComponentClass=null
      PreviewMesh=/BRCosmetics/Animation/Props/Gadgets_QuarterClaspZoom_Cap/Meshes/QuarterClaspZoom_Cap_Gadget
      SkinningMode=SkinOnTheFly
      SoftSourceActor=null
      SourceMode=Default
      UvSetIndex=0
      WholeMeshLOD=-1
      bExcludeBone=false
      bReadDeformedGeometry=true
      bRequireCurrentFrameData=true
      ComponentTags []
      FilteredBones [QuarterClaspZoom_Cap_LOD0_JNT]
      FilteredSockets []
      MeshUserParameter {parameter {name=None, typeDefHandle {registeredTypeIndex=1}, varData []}}
      SamplingRegions []
     }}
   )
  ]
  particleUpdateScript [
   &ParticleState()
   &ScaleMeshSize(Scale Factor=&VectorFromFloat(Value=&FloatFromCurve(FloatCurve=DataInterface() {propertyValues {Curve {defaultValue=1, keys [{tangentMode=RCTM_User, time=0, value=0}, <at 0.0281 1>, <at 0.9 1>, {tangentMode=RCTM_User, time=1, value=0}]}}})))
   &UpdateMeshOrientation(
    Rotation Rate=&RandomRangeFloat(Minimum=1, Maximum=2)
    Rotation Vector=&RandomRangeVector(Minimum=<-1 -1 -1>, Maximum=<1 1 1>)
   )
   &GravityForce()
   &Drag(Drag=1.5, Use Linear Drag=true, Use Rotational Drag=false)
   &SolveForcesAndVelocity(Clamp Velocity=false, Limit Acceleration=false)
  ]
  rendererClasses [&NiagaraMeshRendererProperties]
  renderers [{rendererClass=&NiagaraMeshRendererProperties, rendererIndex=0}]
 }}
summary {
 systemName=NS_Emote_QuarterClaspZoom_Paper1
 emitters [{emitterName=NE_Paper, rendererClasses [&NiagaraMeshRendererProperties]}]
 userVariables [
  User.bIsFrontEnd {type=NiagaraBool}
  User.bIsFrontEndPreview {type=NiagaraBool}
  User.BurstPaper {type=NiagaraFloat, defaultValue=0, description=""}
 ]
}
